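# Translucent, two-sided unlit LimeRock barrier fence effect, usable on meshes and Niagara mesh particles.
# Panning caustics, flickering grid squares and Fresnel build the tinted glow, with edge masks, scanlines and intersection glow.
# Dynamic parameters drive reveal and world-space ripples that distort UVs, alongside a player-or-camera proximity mask and distance-based vertex-offset culling.
# StaticMeshBarrier switches opacity and cull logic, while forward, mobile and low-quality paths simplify the look.
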Material /LimeRock/Assets/FX/Barriers/M_LimeRock_Barrier_Fence
graph {
 properties {
  bUsedWithMeshDeformer=true
  bUsedWithNiagaraMeshParticles=true
  bUsedWithSkeletalMesh=true
  blendMode=BLEND_Translucent
  pixelDepthOffsetMode=PDOM_Legacy
  shadingModel=MSM_Unlit
  twoSided=true
  usedShadingModels=Unlit
 }
 nodes [
  n1 = Constant3Vector() {constant=<rgba 1 1 0 0>}
  n2 = Distance(A=Multiply(A=NamedRerouteUsage(), B=@n1), B=Multiply(A=@n1, B=ActorPositionWS()))
  n3 = ScalarParameter() {defaultValue=1100, parameterName=CullDistance}
  n4 = Floor(None=Saturate(None=Divide(A=@n2, B=@n3)))
  n5 = OneMinus(None=@n4)
  n6 = NamedRerouteUsage()
  n7 = Constant()
  n8 = Constant() {r=1}
  LinearInterpolate(
   A=Multiply(A=Constant3Vector() {constant=<rgba 0 0 -10000 0>}, B=StaticSwitchParameter(True=@n5, False=@n4) {parameterName=StaticMeshBarrier})
   Alpha=Multiply(A=Ceil(None=@n6), B=If(A=@n6, "A > B"=@n7, "A == B"=@n7, "A < B"=@n8) {constB=1})
  ) {constB=0}
  n9 = VectorParameter() {parameterName=EdgeColor, defaultValue=<rgba 0.0205 0.5535 1 1>}
  n10 = Multiply(A=@n9.RGB) {constB=2}
  n11 = VectorParameter() {parameterName=RippleColor, defaultValue=<rgba 3 0.0306 1.4991 1>}
  n12 = DynamicParameter() {paramNames [Reveal,Ripple,BigRipple,Param4]}
  n13 = NamedRerouteUsage()
  n14 = Add(
   A=Multiply(A=LinearInterpolate(A=@n10, B=@n11.RGB, Alpha=@n12.BigRipple), B=NamedRerouteUsage())
   B=LinearInterpolate(A=NamedRerouteUsage(), B=@n9.RGB, Alpha=Saturate(None=@n13)) {constA=1, constAlpha=0.25}
  )
  n15 = Clamp(None=@n14)
  n16 = ScalarParameter() {defaultValue=1.5, parameterName=HDRInensity}
  n17 = Constant() {r=0.1}
  n18 = Multiply(A=@n14, B=@n16)
  MaterialFunctionCall(Emissive=QualitySwitch(
    Default=MaterialFunctionCall(LDR Input=@n15, HDR Tint=@n14, HDR Intensity Multiplier=@n16, Desaturate Luminance=@n17) {materialFunction=&DeriveHDRfromLDR}.RemappedRange
    Low=Multiply(A=@n18) {constB=2}
   )) {materialFunction=&MF_FX_EmissiveAdjustments}
  n19 = NamedRerouteDeclaration(None=SphereMask(A=WorldPosition().XYZ, B=NamedRerouteUsage(), Radius=ScalarParameter() {defaultValue=400, parameterName=NearPlayerRadius}) {hardnessPercent=10}) {name=NearMask, nodeColor=<rgba 0 0.3437 1 1>}
  n20 = DotProduct(A=VertexNormalWS(), B=Constant3Vector() {constant=<rgba 0 0 1 0>})
  n21 = Multiply(A=OneMinus(None=Saturate(None=@n20)), B=NamedRerouteUsage())
  n22 = Max(A=ParticleColor().A, B=@n19)
  n23 = ScalarParameter() {defaultValue=0.75, parameterName=MinOpacity}
  n24 = NamedRerouteUsage()
  n25 = MaterialFunctionCall(Float2=TextureCoordinate()) {materialFunction=&BreakOutFloat2Components}
  n26 = Constant() {r=0.5}
  n27 = SphereMask(A=@n25.R, B=@n26, Hardness=ScalarParameter() {defaultValue=0.5, parameterName=CornerHardness}) {attenuationRadius=0.5, hardnessPercent=0}
  n28 = Max(A=LinearInterpolate(A=OneMinus(None=@n27), Alpha=ScalarParameter() {parameterName=Dome}) {constB=0}, B=NamedRerouteUsage())
  n29 = NamedRerouteDeclaration(None=Power(Base=Max(A=@n24, B=@n28), Exp=ScalarParameter() {defaultValue=3, parameterName=MaskPower}) {constExponent=3}) {name=Masks, nodeColor=<rgba 0 1 0.6953 1>}
  n30 = NamedRerouteUsage()
  n31 = TextureCoordinate()
  n32 = Multiply(A=Time()) {constB=0.01}
  n33 = Multiply(A=ComponentMask(None=Add(A=@n31, B=@n32)) {g=true}, B=ScalarParameter() {defaultValue=200, parameterName=ScanlineTile})
  n34 = LinearInterpolate(A=Subtract(A=Clamp(None=@n23, Min=@n29), B=@n30), B=Add(A=@n30), Alpha=ConstantBiasScale(None=Sine(None=@n33)))
  n35 = Constant() {r=1}
  n36 = ScalarParameter() {defaultValue=1, parameterName=OverallOpacity}
  n37 = VectorParameter() {parameterName=Tiles, defaultValue=<rgba 4 4 1 2>}
  n38 = Multiply(A=AppendVector(A=@n37.B, B=@n37.B), B=TextureCoordinate())
  n39 = NamedRerouteUsage()
  n40 = Multiply(A=Add(A=Time(), B=Multiply(A=@n39) {constB=2}), B=ScalarParameter() {defaultValue=0.4, parameterName=PanSpeed})
  n41 = TextureSample(UVs=Add(A=NamedRerouteUsage(), B=Panner(Coordinate=@n38, Time=@n40) {speedX=-0.02, speedY=0.2})) {texture=/Game/Effects/Fort_Effects/Textures/Tile/T_ShieldCaustics_Tile_Chromatic}
  n42 = Constant() {r=-1}
  n43 = VectorParameter() {parameterName=Tiles, defaultValue=<rgba 4 4 1 2>}
  n44 = TextureSample(UVs=Multiply(A=TextureCoordinate(), B=AppendVector(A=@n43.R, B=@n43.R))) {samplerType=SAMPLERTYPE_LinearColor, texture=/Game/Gadgets/Assets/Teleporter/Textures/T_Teleporter_GridSquares_01}
  n45 = Multiply(A=NamedRerouteUsage()) {constB=12}
  n46 = Add(A=Multiply(A=Time(), B=ScalarParameter() {defaultValue=1, parameterName=PixelChangeSpeed}), B=@n45)
  n47 = Add(A=@n44.B, B=@n46)
  n48 = Multiply(A=ConstantBiasScale(None=Sine(None=@n47) {period=1.4})) {constB=0.4}
  n49 = Add(A=@n44.R, B=@n46)
  n50 = Multiply(A=ConstantBiasScale(None=Sine(None=@n49) {period=0.723}), B=@n45)
  n51 = Max(
   A=Multiply(A=@n44.G, B=Add(A=@n48, B=@n50))
   B=LinearInterpolate(B=ScalarParameter() {defaultValue=0.1, parameterName=PixelMin}, Alpha=NamedRerouteUsage()) {constA=1}
  ) {constB=0.27}
  n52 = Fresnel(ExponentIn=ScalarParameter() {defaultValue=1, parameterName=EdgeFadeExp}) {baseReflectFraction=0, exponent=1}
  n53 = Power(Base=LinearInterpolate(Alpha=@n52) {constA=0.1}) {constExponent=1.7}
  n54 = Add(A=LinearInterpolate(A=Desaturation(None=@n41.RGB, Fraction=@n42), B=@n51) {constAlpha=0.6}, B=@n53)
  n55 = Saturate(None=Add(A=Multiply(A=@n51) {constB=0.5}) {constB=0.1})
  n56 = ShadingPathSwitch(Default=Max(A=Saturate(None=@n54), B=@n53), Forward=@n55, Mobile=@n55)
  n57 = ScalarParameter() {defaultValue=0.4, parameterName=MinBrightness}
  n58 = VectorParameter() {parameterName=OuterColor, defaultValue=<rgba 0.2044 0.4173 1 0>}
  n59 = VectorParameter() {parameterName=InnerColor, defaultValue=<rgba 0.0677 0.5055 1 1>}
  n60 = NamedRerouteDeclaration(None=Multiply(A=LinearInterpolate(A=@n56, Alpha=@n57), B=LinearInterpolate(A=@n58.RGB, B=@n59.RGB, Alpha=@n52))) {name=EmissiveTexture, nodeColor=<rgba 0.4531 1 0 1>}
  n61 = QualitySwitch(Default=DepthFade(Opacity=@n60, FadeDistance=ScalarParameter() {defaultValue=25, parameterName=DephFade}), Low=@n60)
  Multiply(
   A=StaticSwitchParameter(True=@n19, False=Multiply(A=@n21, B=@n22) {constA=1}) {parameterName=StaticMeshBarrier}
   B=Multiply(A=ShadingPathSwitch(Default=@n34, Forward=@n35, Mobile=@n35), B=Multiply(A=@n36, B=@n61))
  ) {constA=1}
  n62 = CollectionParameter() {collection=&FortniteMaterialParameters, parameterName=PlayerPosition}
  n63 = ComponentMask(None=@n62) {b=true, g=true, r=true}
  n64 = CameraPositionWS()
  NamedRerouteDeclaration(None=If(A=Length(None=@n63) {is=class MaterialExpressionLength}, "A > B"=@n63, "A == B"=@n64, "A < B"=@n64) {equalsThreshold=1}) {name=Player/CameraPosition, nodeColor=<rgba 1 0 0.7734 1>}
  n65 = LinearInterpolate(Alpha=DepthFade(FadeDistance=ScalarParameter() {defaultValue=100, parameterName=IntersectGlowSize})) {constA=1.15, constB=0}
  n66 = Constant()
  NamedRerouteDeclaration(None=ShadingPathSwitch(Default=QualitySwitch(Default=@n65, Low=@n66), Forward=@n66, Mobile=@n66)) {name=IntersectGlow, nodeColor=<rgba 0.0312 1 0 1>}
  n67 = Subtract(A=Power(Base=@n25.G))
  n68 = NamedRerouteUsage()
  NamedRerouteDeclaration(None=SphereMask(A=Add(A=@n67, B=@n68), B=@n26) {attenuationRadius=0.5, hardnessPercent=0.25}) {name=VerticalFade, nodeColor=<rgba 0 0.0859 1 1>}
  NamedRerouteDeclaration(None=@n12.Reveal) {name=RevealTime, nodeColor=<rgba 0.1484 0 1 1>}
  n69 = WorldPosition()
  n70 = DynamicParameter() {parameterIndex=1, paramNames [X,Y,Z,Param4]}
  n71 = MaterialFunctionCall(X=@n70.X, Y=@n70.X, Z=@n70.X) {materialFunction=&MakeFloat3}
  n72 = LinearInterpolate(B=ScalarParameter() {defaultValue=3000, parameterName=RippleRadius}, Alpha=@n12.BigRipple) {constA=2000}
  n73 = NamedRerouteUsage()
  n74 = SphereMask(A=@n69.XYZ, B=@n71.Result, Radius=Multiply(A=@n72, B=@n73)) {hardnessPercent=0}
  n75 = Multiply(A=OneMinus(None=NamedRerouteUsage()), B=Multiply(A=@n74) {constB=0.1})
  NamedRerouteDeclaration(None=@n75) {name=RippleFlash, nodeColor=<rgba 1 0 0.5625 1>}
  n76 = Sine(None=@n74)
  n77 = Power(Base=Multiply(A=Saturate(None=@n76), B=@n74), Exp=LinearInterpolate(Alpha=@n73) {constA=1, constB=5}) {constExponent=3}
  NamedRerouteDeclaration(None=Add(A=@n75, B=Multiply(A=@n77) {constB=1.5})) {name=Ripple, nodeColor=<rgba 0 0.6484 1 1>}
  n78 = Subtract(A=@n69.XYZ, B=@n71.Result)
  n79 = Divide(
   A=ComponentMask(None=Transform(None=@n78) {transformSourceType=TRANSFORMSOURCE_World, transformType=TRANSFORM_Tangent}) {g=true, r=true}
   B=@n72
  )
  n80 = ScalarParameter() {defaultValue=-15, parameterName=RippleUVDistort}
  NamedRerouteDeclaration(None=Multiply(A=Multiply(A=@n79, B=@n80), B=@n77)) {name=RippleUV, nodeColor=<rgba 1 0.5391 0 1>}
  NamedRerouteDeclaration(None=@n12.Ripple) {name=RippleTime, nodeColor=<rgba 1 0 0.6797 1>}
  n81 = Fmod(A=Time(), B=Constant() {r=1})
  n82 = ScalarParameter() {parameterName=RippleTime}
  Saturate(None=Divide(A=Subtract(A=@n81, B=@n82), B=ScalarParameter() {defaultValue=1, parameterName=RippleDuration}))
  VectorParameter() {parameterName=RipplePosition2, defaultValue=<rgba -91326 -94560 1632 0>}
  VectorParameter() {parameterName=RipplePosition1, defaultValue=<rgba -91195 -94334 1695 0>}
  VectorParameter() {parameterName=RipplePosition3, defaultValue=<rgba -91600 -94900 1500 0>}
 ]
}
