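# Niagara barrier effect for a LimeRock fence: a translucent cube mesh with four glowing corner flares and a flare that follows the player.
# User Opacity, Ripple, BigRipple and RipplePosition drive the cube material, and Ripple also pulses flare brightness.
# FollowFlare sprites ease toward the system's PlayerPosition, clamped to the box bounds, fading with distance from the player position captured at spawn.
# An alternative Bubble mesh emitter is disabled.
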NiagaraSystem /LimeRock/Assets/FX/Barriers/NS_LimeRock_Barrier_Fence
dependencies {
 usedDataInterfaces [&NiagaraDataInterfaceCurve, &NiagaraDataInterfaceArrayFloat3, &NiagaraDataInterfaceCamera]
 usedDynamicInputs [
  /Niagara/DynamicInputs/TypeConversions/MakeCustomPositionFromBool
  &VectorLength
  &MakeVectorFromLinearColorRGB
  &ReturnCameraPosition
  &NormalizeVector
  &Subtract_Position
 ]
 usedModules [
  &SystemState
  &self:SystemSpawnScript.SystemScriptSource.SystemScriptGraph.NiagaraNodeAssignment_0.SetVariables_#1
  &self:SystemSpawnScript.SystemScriptSource.SystemScriptGraph.NiagaraNodeAssignment_1.SetVariables_#2
 ]
}
emitters {
 Bubble {
  bEnabled=false
  emitterUpdateScript [
   &EmitterState(
    Inactive Response=::NewEnumerator1
    Life Cycle Mode=::NewEnumerator1
    Loop Behavior=::NewEnumerator1
    Loop Duration Mode=::NewEnumerator1
    UseLoopDelay=false
   )
   &SpawnBurst_Instantaneous(Use Spawn Probability=false, UseLoopCountLimit=false)
  ]
  particleSpawnScript [&InitializeParticle(
    Lifetime=0.5
    Mesh Scale=<6.8 6.8 7.5>
    Mesh Scale Mode=::NewEnumerator4
    Position Offset=<0 0 -250>
    UsePositionOffset=true
    Write Material Random=true
   )]
  particleUpdateScript [
   &ParticleState(Kill Particles When Lifetime Has Elapsed=false)
   &DynamicMaterialParameters(
    Index 0 Param 1=&FloatFromCurve(FloatCurve=DataInterface() {propertyValues {Curve {
        defaultValue=1
        keys [
         {arriveTangent=1.8276, interpMode=RCIM_Cubic, leaveTangent=1.8276, tangentMode=RCTM_User, time=0, value=0}
         {tangentMode=RCTM_User, time=1, value=1}
        ]
       }}})
    Index 0 Param 2=$User.Ripple
    Index 0 Param 4=1
    Module.Index0Float0Write=true
    Module.Index0Float1Write=true
    Module.Index0Float2Write=true
    Module.Index0Float3Write=true
   )
   &ScaleColor(Scale Alpha=$User.Opacity, ScaleA=true, ScaleRGB=false)
  ]
  rendererClasses [&NiagaraMeshRendererProperties]
  renderers [{rendererClass=&NiagaraMeshRendererProperties, rendererIndex=0}]
 }
 Cube {
  emitterUpdateScript [
   &EmitterState(
    Inactive Response=::NewEnumerator1
    Life Cycle Mode=::NewEnumerator1
    Loop Behavior=::NewEnumerator1
    Loop Duration Mode=::NewEnumerator1
    UseLoopDelay=false
   )
   &SpawnBurst_Instantaneous(Use Spawn Probability=false, UseLoopCountLimit=false)
  ]
  particleSpawnScript [
   &InitializeParticle(Lifetime=0.5, Mesh Scale=<3.56 3.56 4>, Mesh Scale Mode=::NewEnumerator4, UsePositionOffset=false, Write Material Random=true)
   &CameraOffset(Camera Offset Amount=15)
  ]
  particleUpdateScript [
   &ParticleState(Kill Particles When Lifetime Has Elapsed=false)
   &DynamicMaterialParameters(
    Index 0 Param 1=&FloatFromCurve(FloatCurve=DataInterface() {propertyValues {Curve {
        defaultValue=1
        keys [
         {arriveTangent=1.8276, interpMode=RCIM_Cubic, leaveTangent=1.8276, tangentMode=RCTM_User, time=0, value=0}
         {tangentMode=RCTM_User, time=1, value=1}
        ]
       }}})
    Index 0 Param 2=$User.Ripple
    Index 0 Param 3=$User.BigRipple
    Index 1 Vector 1:Vector3f=$User.RipplePosition
    Module.Index0Float0Write=true
    Module.Index0Float1Write=true
    Module.Index0Float2Write=true
    Module.Index0Float3Write=true
    Module.Index1Float3Write=false
    Module.Index1VectorWrite=true
    Write Parameter Index 1=::NewEnumerator3
   )
   &ScaleColor(Scale Alpha=$User.Opacity, ScaleA=true, ScaleRGB=false)
  ]
  rendererClasses [&NiagaraMeshRendererProperties]
  renderers [{rendererClass=&NiagaraMeshRendererProperties, rendererIndex=0}]
 }
 Flares {
  emitterUpdateScript [
   &EmitterState(
    Inactive Response=::NewEnumerator1
    Life Cycle Mode=::NewEnumerator1
    Loop Behavior=::NewEnumerator1
    Loop Duration Mode=::NewEnumerator1
    UseLoopDelay=false
   )
   &SpawnBurst_Instantaneous(Spawn Count=4, Use Spawn Probability=false, UseLoopCountLimit=false)
  ]
  particleSpawnScript [
   &PartitionParticles(Number Of Partitions=4)
   &InitializeParticle(
    Color=<rgba 0.0926 0.517 1 1>
    Lifetime=5
    Position Offset=&Multiply_Vector(
     A=&SelectVectorFromArray(
      Direct Array Index=$Particles.PartitionParticles.Partition
      Vector Selection Array=DataInterface() {propertyValues {FloatData [<1 1 -1>, <-1 1 -1>, <-1 -1 -1>, <1 -1 -1.4>]}}
     )
     B=<455 455 264>
    )
    Sprite Size Mode=::NewEnumerator3
    Uniform Sprite Size=230
    UsePositionOffset=true
    Write Material Random=true
   )
   &CameraOffset(Camera Offset Amount=20)
   &DynamicMaterialParameters(
    Index 0 Param 1=0.4
    Index 0 Param 2=&RandomRangeFloat(Minimum=1, Maximum=1.5)
    Module.Index0Float0Write=true
    Module.Index0Float1Write=true
    Module.Index0Float2Write=false
    Module.Index0Float3Write=false
   )
   /SetVariables(Particles.SpriteSize2=<50 500>) {isSetParameters=true}
   &SpriteFacingAndAlignment(Sprite Alignment=<0 0 1>, WriteAlignment=true, WriteFacing=false)
  ]
  particleUpdateScript [
   &ParticleState(Kill Particles When Lifetime Has Elapsed=false)
   &ScaleColor(
    Scale RGB=&VectorFromFloat(Value=&FloatFromCurve(
      FloatCurve=DataInterface() {propertyValues {Curve {keys [
          <at 0 1 RCIM_Cubic>
          {arriveTangent=-5.4693, interpMode=RCIM_Cubic, leaveTangent=-5.4693, tangentMode=RCTM_User, time=0.2093, value=5}
          <at 1 1 RCIM_Cubic>
         ]}}}
      CurveIndex=$User.Ripple
     ))
    ScaleA=false
    ScaleRGB=true
   )
   /SetVariables(Particles.LineColor:LinearColor=&MultiplyLinearColor(Linear Color=$Particles.Color, Scale Factor=<1.2 0.7 1 0.25>)) {isSetParameters=true}
   /SetVariables(Particles.LightColor=&Multiply_LinearColorByFloat(Linear Color=$Particles.Color, Float=25)) {isSetParameters=true}
  ]
  rendererClasses [&NiagaraSpriteRendererProperties, &NiagaraLightRendererProperties]
  renderers [
   {rendererClass=&NiagaraSpriteRendererProperties}
   {rendererClass=&NiagaraLightRendererProperties, rendererIndex=1}
   {rendererClass=&NiagaraSpriteRendererProperties, rendererIndex=2}
  ]
 }
 FollowFlare {
  emitterUpdateScript [
   &EmitterState()
   &SpawnRate(SpawnRate=5, Use Spawn Probability=false)
  ]
  particleSpawnScript [
   /SetVariables(Particles.LocalPlayerPosition:NiagaraPosition=&TransformPosition(Position=$System.PlayerPosition, Source Space=::World, Destination Space=::Local)) {isSetParameters=true}
   &InitializeParticle(
    Color=<rgba 0.0724 0.3516 1 1>
    Lifetime=0.7
    Mesh Scale Mode=::NewEnumerator3
    Mesh Uniform Scale=5
    Position Offset=/Niagara/DynamicInputs/Math/Clamp_Vector(
     Vector:Vector3f=Linked() {linkedVariable {name=Particles.LocalPlayerPosition, type=NiagaraPosition}}
     Min:Vector3f=&VectorFromFloat(Value=-464)
     Max:Vector3f=&VectorFromFloat(Value=464)
    )
    Sprite Rotation Mode=::NewEnumerator2
    Sprite Size Mode=::NewEnumerator3
    Uniform Sprite Size=150
    UsePositionOffset=true
    Write Material Random=true
   )
   &CameraOffset(Camera Offset Amount=15)
   /SetVariables(Particles.SmallSpriteParams:Vector4f=<2 0.2 0 0>) {isSetParameters=true}
  ]
  particleUpdateScript [
   &ParticleState()
   &FN_Radial_Material(
    Base Sprite Tile U=2
    Radial Mask Offset X=0.25
    Radial Mask Power=1.5
    Radial Mask Tile U=0.75
    Sprite Extrusion Speed=1
    Sprite Spin Speed=0.2
    TIme:NiagaraFloat=&Add_Float(
     A=&RandomRangeFloat(Minimum=0, Maximum=1)
     B=&FloatFromCurve(
      FloatCurve=DataInterface() {propertyValues {Curve {keys [
          {arriveTangent=1.6374, interpMode=RCIM_Cubic, leaveTangent=1.6374, tangentMode=RCTM_User, time=0, value=0}
          {arriveTangent=0.5763, interpMode=RCIM_Cubic, leaveTangent=0.5763, tangentMode=RCTM_User, time=1, value=1}
         ]}}}
      Scale Curve=0.25
     )
    )
   )
   &SpriteRotationRate(Rotation Rate=20)
   &ScaleSpriteSize(Uniform Curve Sprite Scale=DataInterface() {propertyValues {Curve {keys [
        {arriveTangent=0.7785, interpMode=RCIM_Cubic, leaveTangent=0.7785, tangentMode=RCTM_User, time=0, value=0.5}
        {arriveTangent=0.1871, tangentMode=RCTM_User, time=1, value=1}
       ]}}})
   /SetVariables(Particles.SmallSpriteSize:Vector2f=&Multiply_Vector2D_ByFloat(Float=0.6, Vector2D=$Particles.SpriteSize)) {isSetParameters=true}
   &self:SetClampedDirection(Direction:Vector3f=$System.DirectionToPlayer) {module=SetClampedDirection}
   &SpriteFacingAndAlignment(Sprite Facing=$Output.SetClampedDirection.ClampedDirection, WriteAlignment=false, WriteFacing=true)
   &UpdateMeshOrientation(
    Facing Direction=$Output.SetClampedDirection.ClampedDirection
    Facing Interpolation=false
    Orientation Method=::NewEnumerator1
    Override Quat=false
    TransformFacing=false
    TransformUp=false
   )
   &InterpolateOverTime(
    Rate Of Change:NiagaraFloat=&FloatFromCurve(
     FloatCurve=DataInterface() {propertyValues {Curve {keys [{arriveTangent=-1.3589, interpMode=RCIM_Cubic, leaveTangent=-1.3589, tangentMode=RCTM_User, time=0, value=1}, <at 1 0.1 RCIM_Cubic>]}}}
     Scale Curve=14
    )
    Target Vector Value:Vector3f=&AddVectorToPosition(
     Position=&TransformPosition(Position=$System.PlayerPosition, Source Space=::World, Destination Space=::Local)
     Vector=<0 0 20>
    )
    Variable Type=::NewEnumerator2
   )
   /SetVariables(Particles.Position=&Clamp_Vector(
     Vector:Vector3f=$Particles.InterpolateOverTime.InterpolateOverTime_V3_.Moving Average
     Min:Vector3f=&VectorFromFloat(Value=-464)
     Max:Vector3f=&VectorFromFloat(Value=464)
    )) {isSetParameters=true}
   &ScaleColor(
    Scale Alpha=&Multiply_Float(
     A=&FloatFromCurve(
      FloatCurve=DataInterface() {propertyValues {Curve {keys [<at 50 1 RCIM_Cubic>,<at 120 0>]}}}
      CurveIndex=&DistanceBetweenPositions(A=$Particles.LocalPlayerPosition, B=$Particles.Position)
     )
     B=&FloatFromCurve(FloatCurve=DataInterface() {propertyValues {Curve {
         defaultValue=1
         keys [{arriveTangent=8.8408, interpMode=RCIM_Cubic, time=0, value=0}, <at 0.2 1 RCIM_Cubic>, {tangentMode=RCTM_User, time=1, value=0}]
        }}})
    )
    ScaleA=true
    ScaleRGB=false
   )
   /SetVariables(Particles.VisibilityTag=&MakeCustomIntFromBool(
     Bool=&SetBoolByFloatComparison(
      A=&MakeFloatFromLinearColor(LinearColor=$Particles.Color, Channel:ENiagaraFloatFromLinearColorOptions=::NewEnumerator3)
      B=0.05
      Comparison Type=::NewEnumerator6
     )
     True INT=1
     False INT=0
    )) {isSetParameters=true}
  ]
  rendererClasses [&NiagaraSpriteRendererProperties]
  renderers [
   {rendererClass=&NiagaraSpriteRendererProperties}
   {rendererClass=&NiagaraSpriteRendererProperties, rendererIndex=1}
  ]
 }
}
summary {
 systemName=NS_LimeRock_Barrier_Fence
 emitters [
  {emitterName=Cube, rendererClasses [&NiagaraMeshRendererProperties]}
  {emitterName=Flares, rendererClasses [&NiagaraSpriteRendererProperties, &NiagaraLightRendererProperties]}
  {bEnabled=false, emitterName=Bubble, rendererClasses [&NiagaraMeshRendererProperties]}
  {emitterName=FollowFlare, rendererClasses [&NiagaraSpriteRendererProperties]}
 ]
 userVariables [
  User.BigRipple {type=NiagaraFloat, defaultValue=0, description=""}
  User.Opacity {type=NiagaraFloat, defaultValue=0, description=""}
  User.Ripple {type=NiagaraFloat, defaultValue=0, description=""}
  User.RipplePosition {type=Vector3f, description="", defaultValue {struct=Vector3f, value=<0 0 0>}}
 ]
}
