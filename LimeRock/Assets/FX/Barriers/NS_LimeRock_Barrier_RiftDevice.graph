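# Niagara burst effect for the LimeRock barrier rift device: a brief cyan electrical discharge with flare and mesh blast.
# Electricity spawns staggered bursts partitioned into five-particle ribbons toward rotating array-picked endpoints, with jitter widening outward.
# An index-offset age curve sends a pulse along each ribbon, boosting width and color.
# Flare_Complex adds animated flare sprites and a light, while Minimal stretches forward-offset meshes with curve-driven material parameters and fade.
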NiagaraSystem /LimeRock/Assets/FX/Barriers/NS_LimeRock_Barrier_RiftDevice
dependencies {
 usedDataInterfaces [
  &NiagaraDataInterfaceVectorCurve
  &NiagaraDataInterfaceCurve
  &NiagaraDataInterfaceArrayFloat3
  &NiagaraDataInterfaceVector2DCurve
 ]
 usedDynamicInputs [&RandomRangeFloat]
 usedModules [&SystemState]
}
emitters {
 Electricity {
  emitterUpdateScript [
   &EmitterState()
   &SpawnBurst_Instantaneous(Spawn Count=10, Use Spawn Probability=false, UseLoopCountLimit=false)
   &SpawnBurst_Instantaneous(Spawn Count=20, Spawn Time=0.1, Use Spawn Probability=false, UseLoopCountLimit=false) {module=SpawnBurst_Instantaneous002}
   &SpawnBurst_Instantaneous(Spawn Count=10, Spawn Time=0.2, Use Spawn Probability=false, UseLoopCountLimit=false) {module=SpawnBurst_Instantaneous001}
  ]
  particleSpawnScript [
   &PartitionParticles(
    Normalized Index In Partition Includes One=true
    Particles Per Partition=5
    Partition Count Mode:ENiagara_InfiniteLoopDuration=::NewEnumerator1
    Partition Mode:ENiagara_PartitionMode_v2=::NewEnumerator1
    Write Partition Index to Ribbon Link Order:ENiagara_AttributeSamplingApplyOutput=::NewEnumerator0
    Write Partition to Ribbon ID=::NewEnumerator0
   )
   &InitializeParticle(
    Color=<rgba 0.0312 0.8096 1 1>
    Lifetime=0.4
    Ribbon Width=&FloatFromCurve(
     FloatCurve=DataInterface() {propertyValues {Curve {keys [<at 0 0.2>,<at 0.1 1>,<at 1 1.25>]}}}
     CurveIndex=$Particles.PartitionParticles.Normalized Index In Partition
     Scale Curve=5
    )
    Ribbon Width Mode=::NewEnumerator1
    Sprite Size Mode=::NewEnumerator3
    UsePositionOffset=false
    Write Material Random=true
   )
  ]
  particleUpdateScript [
   &ParticleState()
   /SetVariables(Particles.EndPoint:Vector3f=&RotateVectorByQuaternion(
     Vector To Rotate=&Multiply_VectorByFloat(
      Vector=&NormalizeVector(
       Vector To Normalize=&SelectVectorFromArray(
        Direct Array Index=$Particles.PartitionParticles.Partition
        Direct Array Mode=::NewEnumerator1
        Vector Selection Array=DataInterface() {propertyValues {FloatData [<1 0 0>, <0.5 0.5 0.5>, <0.6 -0.4 -0.1>, <0.4 0.2 -0.3>, <0.5 -0.3 0.5>, <1 0 -0.1>]}}
       )
       Safe Normalize=true
       Safe Fallback Threshold=0
       Safe Fallback Vector=<0 0 0.001>
      )
      Float=180
     )
     Quaternion=&MakeQuaternion(
      Quaternion From=0
      Axis=<0 1 0>
      Angle=&Multiply_Float(A=$Particles.Age, B=&Lerp_Float(A=0.1, B=-0.2, Alpha=$Particles.PartitionParticles.Normalized Index In Partition))
      From=?
      To=?
      Angle Type=::NewEnumerator1
      Normalize Vectors=false
     )
    )) {isSetParameters=true}
   /SetVariables(Particles.Pulse:NiagaraFloat=&FloatFromCurve(
     FloatCurve=DataInterface() {propertyValues {Curve {
        defaultValue=1
        keys [
         {arriveTangent=10.3786, interpMode=RCIM_Cubic, leaveTangent=10.3786, tangentMode=RCTM_User, time=0, value=0}
         {interpMode=RCIM_Cubic, tangentMode=RCTM_User, time=0.15, value=1}
         {arriveTangentWeight=0.1286, interpMode=RCIM_Cubic, leaveTangentWeight=0.2527, tangentMode=RCTM_User, time=0.35, value=0}
        ]
       }}}
     CurveIndex=&Add_Float(
      A=&Multiply_Float(A=$Particles.PartitionParticles.Normalized Index In Partition, B=-0.5)
      B=$Particles.NormalizedAge
     )
     Scale Curve=&FloatFromCurve(
      FloatCurve=DataInterface() {propertyValues {Curve {
         defaultValue=1
         keys [
          {arriveTangent=8.6346, interpMode=RCIM_Cubic, leaveTangent=8.6346, tangentMode=RCTM_User, time=0, value=0}
          {interpMode=RCIM_Cubic, tangentMode=RCTM_User, time=0.25, value=1}
          {arriveTangentWeight=0.3291, interpMode=RCIM_Cubic, leaveTangentWeight=0.1194, tangentMode=RCTM_User, time=0.7604, value=1}
          {arriveTangent=-3.1649, tangentMode=RCTM_User, time=1, value=0.5}
         ]
        }}}
      CurveIndex=$Particles.PartitionParticles.Normalized Index In Partition
     )
    )) {isSetParameters=true}
   /SetVariables(Particles.Position=&AddVectorToPosition(
     Position=&Lerp_Vector(A=<0 0 0>, B=$Particles.EndPoint, Alpha=$Particles.PartitionParticles.Normalized Index In Partition)
     Vector=<0 0 0>
    )) {isSetParameters=true}
   &ShapeLocation(
    MaskSpawnGroup=false
    Override Seed=false
    Shape Origin=$Particles.Position
    Sphere Radius=&Lerp_Float(A=0, B=30, Alpha=$Particles.PartitionParticles.Normalized Index In Partition)
    Sphere Surface Distribution=1
    UseOwnerScale=true
   )
   &self:ScratchModule_01(EndPos:Vector3f=$Particles.EndPoint) {module=ScratchModule_01}
   &DynamicMaterialParameters(
    Index 0 Param 1=$Particles.ScratchModule_01.X
    Index 0 Param 2=$Particles.ScratchModule_01.Y
    Index 0 Param 3=$Particles.ScratchModule_01.Z
    Index 0 Param 4=&Add_Float(A=$Particles.ScratchModule_01.W, B=1)
    Index 1 Param 1=$Particles.MaterialRandom
    Index 1 Param 2=&FloatFromCurve(FloatCurve=DataInterface() {propertyValues {Curve {
        defaultValue=1
        keys [
         {arriveTangent=1.9171, interpMode=RCIM_Cubic, leaveTangent=1.9171, tangentMode=RCTM_User, time=0, value=0.1}
         {interpMode=RCIM_Cubic, leaveTangent=0.8224, time=1, value=1}
        ]
       }}})
    Index 1 Param 3=0.2
    Index 1 Param 4=&Lerp_Float(
     A=1
     B=0
     Alpha=&FloatFromCurve(FloatCurve=DataInterface() {propertyValues {Curve {
         defaultValue=1
         keys [
          {interpMode=RCIM_Cubic, tangentMode=RCTM_User, time=0.7, value=1}
          {
           arriveTangent=-2.9651
           arriveTangentWeight=0.1286
           interpMode=RCIM_Cubic
           leaveTangent=-2.9651
           leaveTangentWeight=0.2527
           tangentMode=RCTM_User
           time=1
           value=0
          }
         ]
        }}})
    )
    Module.Index0Float0Write=true
    Module.Index0Float1Write=true
    Module.Index0Float2Write=true
    Module.Index0Float3Write=true
    Module.Index1Float0Write=true
    Module.Index1Float1Write=true
    Module.Index1Float2Write=true
    Module.Index1Float3Write=true
    Write Parameter Index 1=::NewEnumerator1
   )
   &ScaleColor(Scale RGB=&Lerp_Vector(A=<1 1 1>, B=<5 0.5 1>, Alpha=$Particles.Pulse), ScaleA=false, ScaleRGB=true)
   &ScaleRibbonWidth(Ribbon Width Scale=&Lerp_Float(A=1, B=4, Alpha=$Particles.Pulse))
   /SetVariables(
    Particles.Airglow_Color:LinearColor=&MultiplyLinearColor(Linear Color=$Particles.Color, Scale Factor=<1.2 0.7 1 0.25>)
    Particles.Airglow_Width:NiagaraFloat=&Multiply_Float(A=$Particles.RibbonWidth, B=3)
   ) {isSetParameters=true}
  ]
  rendererClasses [&NiagaraSpriteRendererProperties, &NiagaraRibbonRendererProperties]
  renderers [
   {rendererClass=&NiagaraSpriteRendererProperties}
   {rendererClass=&NiagaraRibbonRendererProperties, rendererIndex=1}
   {rendererClass=&NiagaraRibbonRendererProperties, rendererIndex=2}
  ]
 }
 Flare_Complex {
  emitterUpdateScript [
   &EmitterState()
   &SpawnBurst_Instantaneous(Use Spawn Probability=false, UseLoopCountLimit=false)
  ]
  particleSpawnScript [
   &InitializeParticle(
    Color=<rgba 2 2 2 1>
    Lifetime=0.5
    Sprite Size Mode=::NewEnumerator3
    Uniform Sprite Size=150
    UsePositionOffset=false
    Write Material Random=true
   )
   &CameraOffset(Camera Offset Amount=15)
  ]
  particleUpdateScript [
   &ParticleState()
   &ScaleSpriteSize(Uniform Curve Sprite Scale=DataInterface() {propertyValues {Curve {keys [
        {arriveTangent=5.8052, interpMode=RCIM_Cubic, leaveTangent=5.8052, tangentMode=RCTM_User, time=0, value=0}
        {arriveTangent=-0.9873, interpMode=RCIM_Cubic, leaveTangent=-0.9873, tangentMode=RCTM_User, time=0.4, value=1.25}
        {interpMode=RCIM_Cubic, leaveTangent=0.7472, time=1, value=1}
       ]}}})
   &DynamicMaterialParameters(
    Index 0 Param 1=0.6
    Index 0 Param 2=&FloatFromCurve(FloatCurve=DataInterface() {propertyValues {Curve {
        defaultValue=1
        keys [
         {arriveTangent=-1.5234, interpMode=RCIM_Cubic, leaveTangent=-1.5234, leaveTangentWeight=0.3333, tangentMode=RCTM_User, time=0, value=1.5}
         {tangentMode=RCTM_User, time=1, value=0.1}
        ]
       }}})
    Index 3 Param 1=2
    Index 3 Param 2=5
    Module.Index0Float0Write=true
    Module.Index0Float1Write=true
    Module.Index0Float2Write=true
    Module.Index0Float3Write=false
    Module.Index3Float0Write=true
    Module.Index3Float1Write=true
    Module.Index3Float2Write=false
    Module.Index3Float3Write=false
    Write Parameter Index 3=::NewEnumerator1
   )
   &ScaleColor(
    Scale Alpha=&FloatFromCurve(FloatCurve=DataInterface() {propertyValues {Curve {keys [
         {arriveTangent=14.2117, interpMode=RCIM_Cubic, leaveTangent=14.2117, tangentMode=RCTM_User, time=0, value=0}
         {arriveTangentWeight=0.0943, interpMode=RCIM_Cubic, leaveTangentWeight=0.3771, tangentMode=RCTM_User, time=0.2, value=1}
         <at 1 0>
        ]}}})
    Scale RGB=&MakeVectorFromLinearColorRGB(Color=<rgba 0.0772 0.5401 1 1>)
    ScaleA=true
    ScaleRGB=true
   )
   /SetVariables(Particles.Wide Flare Scale:Vector2f=&Vector2DFromCurve(
     Vector2Curve=DataInterface() {propertyValues {
       XCurve {keys [
         {arriveTangent=4.673, interpMode=RCIM_Cubic, leaveTangent=4.673, tangentMode=RCTM_User, time=0, value=0}
         {arriveTangentWeight=0.1536, interpMode=RCIM_Cubic, leaveTangentWeight=0.1152, tangentMode=RCTM_User, time=0.3, value=1}
         {arriveTangent=-3.1914, interpMode=RCIM_Cubic, leaveTangent=-3.1914, tangentMode=RCTM_User, time=0.5, value=0.15}
         {arriveTangent=-2.3891, interpMode=RCIM_Cubic, leaveTangent=-2.3891, tangentMode=RCTM_User, time=0.7, value=0.25}
         {interpMode=RCIM_Cubic, leaveTangent=-1.231, time=1, value=0.2}
        ]}
       YCurve {keys [
         {arriveTangent=10.6903, interpMode=RCIM_Cubic, leaveTangent=10.6903, tangentMode=RCTM_User, time=0, value=0}
         {arriveTangent=0.8, interpMode=RCIM_Cubic, leaveTangent=0.8, time=0.25, value=1}
         {arriveTangent=-1.7663, interpMode=RCIM_Cubic, leaveTangent=-1.7663, tangentMode=RCTM_User, time=0.5, value=0.4}
         {arriveTangent=-4.8806, interpMode=RCIM_Cubic, leaveTangent=-4.8806, tangentMode=RCTM_User, time=0.7, value=1}
         {interpMode=RCIM_Cubic, leaveTangent=-2.2148, time=1, value=0.4}
        ]}
      }}
     ScaleCurve=<1200 300>
    )) {isSetParameters=true}
   /SetVariables(
    Particles.LightColor=&MultiplyLinearColor(Linear Color=$Particles.Color, Scale Factor=<2 0.7 1 100>)
    Particles.LightRadius=600
   ) {isSetParameters=true}
  ]
  rendererClasses [&NiagaraSpriteRendererProperties, &NiagaraLightRendererProperties]
  renderers [
   {rendererClass=&NiagaraSpriteRendererProperties}
   {rendererClass=&NiagaraSpriteRendererProperties, rendererIndex=1}
   {rendererClass=&NiagaraLightRendererProperties, rendererIndex=2}
  ]
 }
 Minimal {
  emitterUpdateScript [
   &EmitterState()
   &SpawnBurst_Instantaneous(Use Spawn Probability=false, UseLoopCountLimit=false)
  ]
  particleSpawnScript [&InitializeParticle(
    Color=<rgba 0.2923 3.2362 5 1>
    Lifetime=0.6
    Mesh Scale=<3 1.5 1.5>
    Mesh Scale Mode=::NewEnumerator4
    Position Offset=<200 0 0>
    UsePositionOffset=true
    Write Material Random=true
   )]
  particleUpdateScript [
   &ParticleState()
   &ScaleMeshSize(Scale Factor=&VectorFromCurve(VectorCurve=DataInterface() {propertyValues {
       XCurve {
        defaultValue=1
        keys [
         {arriveTangent=2.8113, interpMode=RCIM_Cubic, leaveTangent=2.8113, tangentMode=RCTM_User, time=0, value=0}
         {arriveTangent=0.3634, tangentMode=RCTM_User, time=1, value=1}
        ]
       }
       YCurve {keys [
         {arriveTangent=2.7611, interpMode=RCIM_Cubic, leaveTangent=2.7611, tangentMode=RCTM_User, time=0, value=0.15}
         {arriveTangentWeight=0.174, interpMode=RCIM_Cubic, leaveTangentWeight=0.174, tangentMode=RCTM_User, time=0.6, value=1}
         {arriveTangent=-0.3811, interpMode=RCIM_Cubic, leaveTangent=-0.3811, tangentMode=RCTM_User, time=1, value=0.9}
        ]}
       ZCurve {keys [
         {arriveTangent=2.7611, interpMode=RCIM_Cubic, leaveTangent=2.7611, tangentMode=RCTM_User, time=0, value=0.15}
         {arriveTangentWeight=0.174, interpMode=RCIM_Cubic, leaveTangentWeight=0.174, tangentMode=RCTM_User, time=0.6, value=1}
         {arriveTangent=-0.3811, interpMode=RCIM_Cubic, leaveTangent=-0.3811, tangentMode=RCTM_User, time=1, value=0.9}
        ]}
      }}))
   &DynamicMaterialParameters(
    Index 0 Param 1=1
    Index 0 Param 2=-2
    Index 1 Param 1=&FloatFromCurve(FloatCurve=DataInterface() {propertyValues {Curve {
        defaultValue=1
        keys [
         {arriveTangent=5.9511, interpMode=RCIM_Cubic, leaveTangent=5.9511, tangentMode=RCTM_User, time=0, value=-2}
         {arriveTangent=1.3531, tangentMode=RCTM_User, time=1, value=1}
        ]
       }}})
    Index 1 Param 2=&FloatFromCurve(FloatCurve=DataInterface() {propertyValues {Curve {keys [<at 0 1 RCIM_Cubic>,<at 1 0>]}}}, Scale Curve=3)
    Module.Index0Float0Write=true
    Module.Index0Float1Write=true
    Module.Index0Float2Write=true
    Module.Index0Float3Write=true
    Module.Index1Float0Write=true
    Module.Index1Float1Write=true
    Module.Index1Float2Write=false
    Module.Index1Float3Write=false
    Write Parameter Index 1=::NewEnumerator1
   )
   &ScaleColor(
    Scale Alpha=&FloatFromCurve(FloatCurve=DataInterface() {propertyValues {Curve {
        defaultValue=1
        keys [
         {arriveTangent=17.5908, interpMode=RCIM_Cubic, leaveTangent=17.5908, tangentMode=RCTM_User, time=-0.0041, value=0}
         {arriveTangent=-1.4064, interpMode=RCIM_Cubic, leaveTangent=-1.4064, tangentMode=RCTM_User, time=0.1, value=1}
         {tangentMode=RCTM_User, time=1, value=0}
        ]
       }}})
    ScaleA=true
    ScaleRGB=true
   )
  ]
  rendererClasses [&NiagaraSpriteRendererProperties, &NiagaraMeshRendererProperties]
  renderers [
   {rendererClass=&NiagaraSpriteRendererProperties}
   {rendererClass=&NiagaraMeshRendererProperties, rendererIndex=1}
   {rendererClass=&NiagaraMeshRendererProperties, rendererIndex=2}
  ]
 }
}
summary {
 systemName=NS_LimeRock_Barrier_RiftDevice
 emitters [
  {emitterName=Minimal, rendererClasses [&NiagaraSpriteRendererProperties, &NiagaraMeshRendererProperties]}
  {emitterName=Electricity, rendererClasses [&NiagaraSpriteRendererProperties, &NiagaraRibbonRendererProperties]}
  {emitterName=Flare_Complex, rendererClasses [&NiagaraSpriteRendererProperties, &NiagaraLightRendererProperties]}
 ]
 userVariables []
}
