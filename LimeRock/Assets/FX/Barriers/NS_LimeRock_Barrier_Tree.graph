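# Niagara system for a LimeRock tree barrier, combining a barrier mesh with magenta glow, drifting motes and falling leaves.
# Glow and Shimmers track the point at barrier radius toward the player, fading with player distance; Shimmers spawn only while the player is in range.
# GPU motes and leaves respond to User.Intensity and a point force at User.PointForceOrigin, which drive spawning, lifetime and motion.
# An arced BarrierRibbon emitter exists but is disabled.
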NiagaraSystem /LimeRock/Assets/FX/Barriers/NS_LimeRock_Barrier_Tree
dependencies {
 usedDataInterfaces [
  &NiagaraDataInterfaceCurve
  &NiagaraDataInterfaceCamera
  &NiagaraDataInterfaceSpriteRendererInfo
 ]
 usedDynamicInputs [
  /Niagara/DynamicInputs/TypeConversions/MakeCustomPositionFromBool
  &SetBoolByFloatComparison
  &VectorLength
  &MakeVectorFromLinearColorRGB
  &Subtract_Position
 ]
 usedModules [
  &self:SystemSpawnScript.SystemScriptSource.SystemScriptGraph.NiagaraNodeAssignment_2.SetVariables_#1
  &SystemState
  &self:SystemSpawnScript.SystemScriptSource.SystemScriptGraph.NiagaraNodeAssignment_1.SetVariables_#2
  &self:SystemSpawnScript.SystemScriptSource.SystemScriptGraph.NiagaraNodeAssignment_0.SetVariables_#3
  &self:SystemSpawnScript.SystemScriptSource.SystemScriptGraph.NiagaraNodeAssignment_3.SetVariables_#4
 ]
}
emitters {
 AmbientMotes {
  simTarget=GPUComputeSim
  emitterUpdateScript [
   &EmitterState()
   /SetVariables(Emitter.CamPosition:NiagaraPosition=&ReturnCameraPosition(
     Coordinate Space:ENiagaraCoordinateSpace=::Simulation
     Camera Query:NiagaraDataInterfaceCamera=DataInterface() {propertyValues {bRequireCurrentFrameData=false}}
    )) {isSetParameters=true}
   &SpawnRate(
    SpawnRate=&Multiply_Float(A=&ClampFloat(Float=&Add_Float(A=$User.Intensity, B=$User.PointForce), Max=1.5), B=450)
    Use Spawn Probability=false
   )
   &SpawnRate(SpawnGroup=1, SpawnRate=&Multiply_Float(A=$User.Intensity, B=350), Use Spawn Probability=false) {module=SpawnRate001}
  ]
  particleSpawnScript [
   &InitializeParticle(
    Color=&MaskLinearColorBySpawnGroup(
     Masked Linear Color=<rgba 0.0214 0.0203 0.026 0.7>
     Passthrough Linear Color=&RandomRangeLinearColor(Minimum=<rgba 1 0.0396 0.978 0.8>, Maximum=<rgba 1.5 0.058 0.5101 1>)
     Spawn Group Mask=1
    )
    Lifetime=&Multiply_Float(
     A=&RandomRangeFloat(Minimum=4, Maximum=5)
     B=&Lerp_Float(
      A=1
      B=0.6
      Alpha=&Add_Float(A=$User.Intensity, B=&Abs_Float(Float:NiagaraFloat=$User.PointForce))
     )
    )
    Sprite Rotation Angle=&MaskFloatBySpawnGroup(Masked Float=&RandomRangeFloat(Minimum=0, Maximum=1), Passthrough Float=0, Spawn Group Mask=1)
    Sprite Rotation Mode=::NewEnumerator1
    Sprite Size Mode=::NewEnumerator1
    Uniform Sprite Size Max=25
    Uniform Sprite Size Min=17
    UsePositionOffset=false
    Write Material Random=true
   )
   &ShapeLocation(
    Apply To Particle Position=::NewEnumerator1
    Cylinder Height=3000
    Cylinder Height Midpoint=0.1
    Cylinder Radius=7000
    MaskSpawnGroup=false
    Offset=<2000 0 0>
    Offset Mode=::NewEnumerator0
    Override Seed=false
    Shape Primitive=::NewEnumerator1
    Surface Only=false
    UseOwnerScale=true
   )
   &ShapeLocation(
    Apply To Particle Position=::NewEnumerator1
    MaskSpawnGroup=false
    Offset=<0 0 200>
    Offset Mode=::NewEnumerator0
    Override Seed=false
    Shape Origin=$System.PlayerPosition
    Sphere Radius=1500
    Sphere Surface Distribution=0.5
    UseOwnerScale=true
   ) {module=ShapeLocation001}
   /SetVariables(Particles.Position=&Lerp_Vector(
     A=Linked() {linkedVariable {name=Particles.ShapeLocation.ShapePosition, type=NiagaraPosition}}
     B=Linked() {linkedVariable {name=Particles.ShapeLocation001.ShapePosition, type=NiagaraPosition}}
     Alpha=&RandomRangeFloat(Minimum=0, Maximum=1)
    )) {isSetParameters=true}
   &AddVelocity(
    Cone Axis=&MaskVectorBySpawnGroup(Masked Vector=<0.5 0 -0.3>, Passthrough Vector=<1 0 0>, Spawn Group Mask=1)
    Module.UseConeFalloff=false
    Module.UseVelDistribution=false
    Override Seed=false
    Velocity Mode=::NewEnumerator2
    Velocity Speed=&Lerp_Float(A=400, B=1000, Alpha=$User.Intensity)
   )
   &PointForce(
    Force Origin:NiagaraPosition=$User.PointForceOrigin
    Force Strength=&Multiply_Float(A=$User.PointForce, B=1000)
    Module.Use Falloff=false
    Module.UseLinearForce=false
   ) {module=PointForce001}
   &ApplyInitialForces()
  ]
  particleUpdateScript [
   &ParticleState()
   &ScaleSpriteSize(
    Uniform Curve Scale=&FloatFromCurve(
     FloatCurve=DataInterface() {propertyValues {Curve {keys [{arriveTangent=0.0004, interpMode=RCIM_Cubic, leaveTangent=0.0004, tangentMode=RCTM_User, time=500, value=0.1}, <at 3000 1>]}}}
     CurveIndex=&DistanceBetweenPositions(A=$Emitter.CamPosition, B=$Particles.Position)
     Scale Curve=&MaskFloatBySpawnGroup(Masked Float=0.6, Passthrough Float=1, Spawn Group Mask=1)
    )
    Uniform Curve Sprite Scale=DataInterface() {propertyValues {Curve {
       defaultValue=0
       keys [
        {interpMode=RCIM_Cubic, tangentMode=RCTM_User, time=0, value=0}
        {arriveTangentWeight=0.0833, interpMode=RCIM_Cubic, leaveTangentWeight=0.1742, tangentMode=RCTM_User, time=0.25, value=1}
        {
         arriveTangent=-0.2912
         arriveTangentWeight=0.2903
         interpMode=RCIM_Cubic
         leaveTangent=-0.2912
         leaveTangentWeight=0.1263
         tangentMode=RCTM_User
         time=0.71
         value=0.9
        }
        {tangentMode=RCTM_User, time=1, value=0}
       ]
      }}}
   )
   &WindForce(
    Frequency:NiagaraFloat=&Lerp_Float(A=1, B=3, Alpha=$User.Intensity)
    Friction Mode:ENiagaraWindFrictionMode=::NewEnumerator0
    Scale:NiagaraFloat=&Lerp_Float(A=1, B=2, Alpha=$User.Intensity)
    Surface Alignment Mode:ENiagaraWindCombingMode=::NewEnumerator0
    Wind Speed=&MaskVectorBySpawnGroup(Masked Vector=<0.4 0 -0.8>, Passthrough Vector=<0.6 0 0.6>, Spawn Group Mask=1)
    Wind Speed Scale=&RandomRangeFloat(Minimum=60, Maximum=100)
   )
   &Drag(
    Drag=&MaskFloatBySpawnGroup(Masked Float=1.5, Passthrough Float=1, Spawn Group Mask=1)
    Use Linear Drag=true
    Use Rotational Drag=false
   )
   &PointForce(
    Force Origin:NiagaraPosition=$User.PointForceOrigin
    Force Strength=&Multiply_Float(A=$User.PointForce, B=5000)
    Module.Use Falloff=false
    Module.UseLinearForce=false
   )
   &SolveForcesAndVelocity(Clamp Velocity=false, Limit Acceleration=false)
   &ScaleSpriteSizeBySpeed(
    Max Scale Factor=&MaskVector2dBySpawnGroup(Masked Vector2D=<1 1>, Passthrough Vector2D=<1 8>, Spawn Group Mask=1)
    Sample Scale Factor By Curve=true
    Scale Factor Curve=DataInterface() {propertyValues {Curve {keys [<at 0.1 0 RCIM_Cubic>,<at 1 1>]}}}
    Velocity Threshold=3000
   )
  ]
  rendererClasses [&NiagaraSpriteRendererProperties]
  renderers [
   {rendererClass=&NiagaraSpriteRendererProperties}
   {rendererClass=&NiagaraSpriteRendererProperties, rendererIndex=1}
  ]
 }
 BarrierMesh {
  emitterUpdateScript [
   &EmitterState(
    Inactive Response=::NewEnumerator1
    Life Cycle Mode=::NewEnumerator1
    Loop Behavior=::NewEnumerator1
    Loop Duration Mode=::NewEnumerator1
    UseLoopDelay=false
   )
   &SpawnBurst_Instantaneous(Use Spawn Probability=false, UseLoopCountLimit=false)
  ]
  particleSpawnScript [&InitializeParticle(
    Color Mode=::NewEnumerator0
    Lifetime=5
    Mesh Scale Mode=::NewEnumerator3
    Position Offset=<0 0 10>
    UsePositionOffset=true
    Write Material Random=true
   )]
  particleUpdateScript [&ParticleState(Kill Particles When Lifetime Has Elapsed=false)]
  rendererClasses [&NiagaraMeshRendererProperties]
  renderers [{rendererClass=&NiagaraMeshRendererProperties, rendererIndex=0}]
 }
 BarrierRibbon {
  bEnabled=false
  emitterUpdateScript [
   &EmitterState(Life Cycle Mode=::NewEnumerator1, Loop Behavior=::NewEnumerator1, Loop Duration Mode=::NewEnumerator1, UseLoopDelay=false)
   &SpawnBurst_Instantaneous(Spawn Count=20, Use Spawn Probability=false, UseLoopCountLimit=false)
  ]
  particleSpawnScript [
   &InitializeParticle(
    Color=<rgba 2 0.3229 1.7474 1>
    Lifetime=5
    Ribbon Width=400
    Ribbon Width Mode=::NewEnumerator1
    Sprite Size Mode=::NewEnumerator3
    Uniform Sprite Size=200
    UsePositionOffset=false
    Write Material Random=true
   )
   /SetVariables(Particles.Position=&RotateVectorByQuaternion(
     Vector To Rotate=&MakeVector(
      X=$System.BarrierRadius
      Y=0
      Z=&Add_Float(
       A=-20
       B=&FloatFromCurve(
        FloatCurve=DataInterface() {propertyValues {Curve {
           defaultValue=1
           keys [
            {arriveTangent=1.8181, interpMode=RCIM_Cubic, leaveTangent=1.8181, tangentMode=RCTM_User, time=0, value=-1}
            {arriveTangent=3.9965, interpMode=RCIM_Cubic, leaveTangent=3.9965, tangentMode=RCTM_User, time=0.247, value=-0.1989}
            {arriveTangentWeight=0.0351, interpMode=RCIM_Cubic, leaveTangentWeight=0.0291, tangentMode=RCTM_User, time=0.2512, value=0}
            {arriveTangentWeight=0.0177, interpMode=RCIM_Cubic, leaveTangentWeight=0.045, tangentMode=RCTM_User, time=0.4985, value=0}
            {arriveTangent=-3.9484, interpMode=RCIM_Cubic, leaveTangent=-3.9484, tangentMode=RCTM_User, time=0.5601, value=-0.1016}
            {arriveTangent=-1.0953, tangentMode=RCTM_User, time=1, value=-0.75}
           ]
          }}}
        CurveIndex=&ReturnNormalizedExecIndex()
        Scale Curve=1100
       )
      )
     )
     Quaternion=&MakeQuaternion(
      Quaternion From=0
      Axis=<0 0 1>
      Angle=&Lerp_Float(A=-0.15, B=0.2, Alpha=&ReturnNormalizedExecIndex())
      From=?
      To=?
      Angle Type=::NewEnumerator1
      Normalize Vectors=false
     )
    )) {isSetParameters=true}
   /SetVariables(Particles.RibbonFacing=&NormalizeVector(
     Vector To Normalize=&Multiply_Vector(A=Linked() {linkedVariable {name=Particles.Position, type=NiagaraPosition}}, B=<1 1 0>)
     Safe Normalize=true
     Safe Fallback Threshold=0
     Safe Fallback Vector=<0 0 0.001>
    )) {isSetParameters=true}
  ]
  particleUpdateScript [
   &ParticleState(Kill Particles When Lifetime Has Elapsed=false)
   &DynamicMaterialParameters(
    Index 0 Param 1=&Waveform(
     Global Phase=&Add_Float(
      A=&Multiply_Float(A=$Emitter.Age, B=0.2)
      B=&ReturnNormalizedExecIndex(Normalized Index Scale=5)
     )
     "Waveform [1]"=::NewEnumerator12
     "[1] Frequency"=4
     "[1] Amplitude Min/Max"=<-1 0.5>
    )
    Module.Index0Float0Write=true
    Module.Index0Float1Write=false
    Module.Index0Float2Write=false
    Module.Index0Float3Write=false
   )
  ]
  rendererClasses [&NiagaraSpriteRendererProperties, &NiagaraRibbonRendererProperties]
  renderers [
   {rendererClass=&NiagaraSpriteRendererProperties}
   {rendererClass=&NiagaraRibbonRendererProperties, rendererIndex=1}
  ]
 }
 Glow {
  emitterUpdateScript [
   &EmitterState(
    Inactive Response=::NewEnumerator1
    Life Cycle Mode=::NewEnumerator1
    Loop Behavior=::NewEnumerator1
    Loop Duration Mode=::NewEnumerator1
    UseLoopDelay=false
   )
   &SpawnBurst_Instantaneous(Use Spawn Probability=false, UseLoopCountLimit=false)
  ]
  particleSpawnScript [
   &InitializeParticle(
    Color=<rgba 0.7526 0.0614 1 0.5>
    Lifetime Min=1.5
    Lifetime Mode=::NewEnumerator1
    Sprite Rotation Mode=::NewEnumerator2
    Sprite Size Mode=::NewEnumerator3
    Uniform Sprite Size=400
    UsePositionOffset=false
    Write Material Random=true
   )
   &DynamicMaterialParameters(
    Index 0 Param 1=2.5
    Index 0 Param 2=2
    Index 0 Param 3=0.5
    Module.Index0Float0Write=true
    Module.Index0Float1Write=true
    Module.Index0Float2Write=true
    Module.Index0Float3Write=false
   )
  ]
  particleUpdateScript [
   &ParticleState(Kill Particles When Lifetime Has Elapsed=false)
   &InterpolateOverTime(
    Rate Of Change=20
    Target Position Value:NiagaraPosition=&AddVectorToPosition(
     Position=$Engine.Owner.Position
     Vector=&Multiply_VectorByFloat(Vector=$System.DirectionToPlayer, Float=$System.BarrierRadius)
    )
    Variable Type=::NewEnumerator6
   )
   /SetVariables(Particles.Position=$Particles.InterpolateOverTime.InterpolateOverTime_Position.Moving Average) {isSetParameters=true, module=SetVariables_#5}
   &SpriteFacingAndAlignment(Sprite Alignment=$System.DirectionToPlayer, Sprite Facing=$System.DirectionToPlayer, WriteAlignment=false, WriteFacing=true)
   &ScaleSpriteSize(
    Uniform Curve Index=&DistanceBetweenPositions(A=$System.PlayerPosition, B=$Particles.Position)
    Uniform Curve Sprite Scale=DataInterface() {propertyValues {Curve {keys [{arriveTangent=0.986, interpMode=RCIM_Cubic, time=0, value=1}, {arriveTangent=-0.0022, tangentMode=RCTM_User, time=600, value=0}]}}}
   )
   &ScaleColor(
    Scale Alpha=&FloatFromCurve(
     FloatCurve=DataInterface() {propertyValues {Curve {keys [
         {arriveTangent=-0.0014, interpMode=RCIM_Cubic, time=0, value=1}
         {arriveTangent=-0.0017, interpMode=RCIM_Cubic, leaveTangent=-0.0017, time=100, value=0.7}
         {arriveTangent=-0.0008, tangentMode=RCTM_User, time=600, value=0}
        ]}}}
     CurveIndex=&DistanceBetweenPositions(A=$System.PlayerPosition, B=$Particles.Position)
    )
    ScaleA=true
    ScaleRGB=false
   )
  ]
  rendererClasses [&NiagaraSpriteRendererProperties]
  renderers [{rendererClass=&NiagaraSpriteRendererProperties}]
 }
 Leaves {
  simTarget=GPUComputeSim
  emitterUpdateScript [
   &EmitterState()
   &SpawnRate(
    SpawnRate=&Add_Float(A=&Multiply_Float(A=$User.PointForce, B=200), B=1)
    Use Spawn Probability=false
   )
  ]
  particleSpawnScript [
   &InitializeParticle(
    Color=<rgba 0.0054 0.0938 0.0029 1>
    Lifetime=&Multiply_Float(
     A=&RandomRangeFloat(Minimum=4, Maximum=5)
     B=&Lerp_Float(
      A=1
      B=0.65
      Alpha=&Add_Float(A=$User.Intensity, B=&Abs_Float(Float:NiagaraFloat=$User.PointForce))
     )
    )
    Sprite Rotation Mode=::NewEnumerator2
    Sprite Size Mode=::NewEnumerator1
    Uniform Sprite Size Max=50
    Uniform Sprite Size Min=35
    UsePositionOffset=false
    Write Material Random=true
   )
   &ShapeLocation(
    MaskSpawnGroup=false
    Non Uniform Scale=<1 1 0.8>
    Offset=<250 0 2150>
    Offset Mode=::NewEnumerator0
    Override Seed=false
    Sphere Radius=700
    Sphere Surface Distribution=0.8
    UseOwnerScale=true
   )
   &PointForce(
    Force Origin:NiagaraPosition=$User.PointForceOrigin
    Force Strength=&Multiply_Float(A=$User.PointForce, B=500)
    Module.Use Falloff=false
    Module.UseLinearForce=false
   ) {module=PointForce001}
   &ApplyInitialForces()
   &DynamicMaterialParameters(
    Index 0 Param 1=&RandomRangeFloat(Minimum=0.6, Maximum=0.8)
    Module.Index0Float0Write=true
    Module.Index0Float1Write=true
    Module.Index0Float2Write=true
    Module.Index0Float3Write=true
   )
  ]
  particleUpdateScript [
   &ParticleState()
   &ScaleSpriteSize(Uniform Curve Sprite Scale=DataInterface() {propertyValues {Curve {
       defaultValue=0
       keys [
        {arriveTangent=10.3157, interpMode=RCIM_Cubic, leaveTangent=10.3157, tangentMode=RCTM_User, time=0, value=0}
        {arriveTangentWeight=0.0833, interpMode=RCIM_Cubic, leaveTangentWeight=0.1533, tangentMode=RCTM_User, time=0.2, value=1}
        {arriveTangentWeight=0.1597, interpMode=RCIM_Cubic, leaveTangentWeight=0.1007, tangentMode=RCTM_User, time=0.6, value=1}
        {arriveTangent=-3.9496, tangentMode=RCTM_User, time=1, value=0}
       ]
      }}})
   &SpriteRotationRate(
    Delta Time:Vector2f=$Engine.DeltaTime
    Rotation Rate=&FloatFromCurve(
     FloatCurve=DataInterface() {propertyValues {Curve {
        defaultValue=0
        keys [
         {arriveTangent=2.0645, interpMode=RCIM_Cubic, leaveTangent=2.0645, tangentMode=RCTM_User, time=0, value=1}
         {arriveTangent=-0.8768, interpMode=RCIM_Cubic, leaveTangent=-0.8768, tangentMode=RCTM_User, time=1, value=2}
         {tangentMode=RCTM_User, time=5, value=0.2}
        ]
       }}}
     CurveIndex=$Particles.Age
     Scale Curve=&FloatFromCurve(
      FloatCurve=DataInterface() {propertyValues {Curve {keys [
          {interpMode=RCIM_Cubic, tangentMode=RCTM_User, time=0, value=-1}
          {arriveTangent=5.8272, interpMode=RCIM_Cubic, leaveTangent=5.8272, tangentMode=RCTM_User, time=0.5, value=0}
          {interpMode=RCIM_Cubic, tangentMode=RCTM_User, time=1, value=1}
         ]}}}
      CurveIndex=&RandomRangeFloat(Minimum=0, Maximum=1)
      Scale Curve=500
     )
    )
   )
   &SubUVAnimation(
    Play Rate:NiagaraFloat=&RandomRangeFloat(Minimum=0.22, Maximum=0.37)
    Playback Scale Curve=DataInterface() {propertyValues {Curve {keys [{tangentMode=RCTM_User, time=0, value=0}, {tangentMode=RCTM_User, time=1, value=1}]}}}
    Random Start Frame=true
    "Sprite Renderer "=DataInterface() {propertyValues {SpriteRenderer=&self:Leaves_0.NiagaraSpriteRendererProperties_2}}
    SubUV Animation Mode:ENiagara_SubUVLookupModeV2=::NewEnumerator3
    UseEndFrame=false
    UsePlayRateScale=false
    UseStartFrame=false
   )
   &CurlNoiseForce(
    Bias Noise=false
    Noise Frequency=&Lerp_Float(A=10, B=70, Alpha=$User.Intensity)
    Noise Strength=&Lerp_Float(A=600, B=3000, Alpha=$User.Intensity)
    Pan Noise=false
   )
   &WindForce(
    Friction Mode:ENiagaraWindFrictionMode=::NewEnumerator0
    Surface Alignment Mode:ENiagaraWindCombingMode=::NewEnumerator0
    Turbulence Mode:ENiagaraWindTurbulenceMode=::NewEnumerator0
    Wind Speed=<1 0 0>
    Wind Speed Scale=&RandomRangeFloat(Minimum=40, Maximum=70)
   )
   &GravityForce(Gravity=<0 0 -400>)
   &Drag(Drag=&RandomRangeFloat(Minimum=2, Maximum=3), Use Linear Drag=true, Use Rotational Drag=false)
   &PointForce(
    Force Origin:NiagaraPosition=$User.PointForceOrigin
    Force Strength=&Multiply_Float(A=$User.PointForce, B=4000)
    Module.Use Falloff=false
    Module.UseLinearForce=false
   )
   &SolveForcesAndVelocity(Clamp Velocity=false, Limit Acceleration=false)
  ]
  rendererClasses [&NiagaraSpriteRendererProperties]
  renderers [{rendererClass=&NiagaraSpriteRendererProperties}]
 }
 Shimmers {
  emitterUpdateScript [
   &EmitterState()
   &SpawnRate(
    SpawnRate=&MakeCustomFloatFromBool(Bool=$System.PlayerInRange, True Float=15, False Float=0)
    Use Spawn Probability=false
   )
  ]
  particleSpawnScript [
   &InitializeParticle(
    Color=<rgba 1.872 0.202 2 1>
    Lifetime Max=1.5
    Lifetime Mode=::NewEnumerator1
    Sprite Rotation Mode=::NewEnumerator2
    Sprite Size Mode=::NewEnumerator1
    Uniform Sprite Size Max=400
    Uniform Sprite Size Min=200
    UsePositionOffset=false
    Write Material Random=true
   )
   &DynamicMaterialParameters(
    Index 0 Param 1=&RandomRangeFloat(Minimum=0, Maximum=1)
    Index 0 Param 2=&RandomRangeFloat(Minimum=0, Maximum=1)
    Index 0 Param 3=0.5
    Module.Index0Float0Write=true
    Module.Index0Float1Write=true
    Module.Index0Float2Write=true
    Module.Index0Float3Write=false
   )
  ]
  particleUpdateScript [
   &ParticleState()
   &SpriteRotationRate(Rotation Rate=&RandomRangeFloat(Minimum=-20, Maximum=20))
   &ScaleSpriteSize(Uniform Curve Sprite Scale=DataInterface() {propertyValues {Curve {keys [
        {arriveTangent=0.986, interpMode=RCIM_Cubic, leaveTangent=0.986, tangentMode=RCTM_User, time=0, value=0.5}
        {tangentMode=RCTM_User, time=1, value=1}
       ]}}})
   &InterpolateOverTime(
    Rate Of Change:NiagaraFloat=&FloatFromCurve(
     FloatCurve=DataInterface() {propertyValues {Curve {keys [<at 0 1 RCIM_Cubic>, {arriveTangent=-1.4433, tangentMode=RCTM_User, time=1, value=0}]}}}
     Scale Curve=&RandomRangeFloat(Minimum=1, Maximum=3)
    )
    Target Position Value:NiagaraPosition=&AddVectorToPosition(
     Position=$Engine.Owner.Position
     Vector=&Add_Vector(
      A=&Multiply_VectorByFloat(Vector=$System.DirectionToPlayer, Float=$System.BarrierRadius)
      B=&RotateVectorByQuaternion(
       Vector To Rotate=&RandomRangeVector(Minimum=<0 200 0>, Maximum=<0 0 0>)
       Quaternion=&MakeQuaternion(
        Quaternion From=0
        Axis=$System.DirectionToPlayer
        Angle=&RandomRangeFloat(Minimum=0, Maximum=1)
        From=?
        To=?
        Angle Type=::NewEnumerator1
        Normalize Vectors=false
       )
      )
     )
    )
    Variable Type=::NewEnumerator6
   )
   /SetVariables(Particles.Position=$Particles.InterpolateOverTime.InterpolateOverTime_Position.Moving Average) {isSetParameters=true, module=SetVariables_#5}
   &SpriteFacingAndAlignment(Sprite Alignment=$System.DirectionToPlayer, Sprite Facing=$System.DirectionToPlayer, WriteAlignment=false, WriteFacing=true)
   &ScaleColor(
    Scale Alpha=&FloatFromCurve(
     FloatCurve=DataInterface() {propertyValues {Curve {keys [{arriveTangent=1.0833, interpMode=RCIM_Cubic, leaveTangent=1.0833, time=0, value=0}, <at 0.6 0.65 RCIM_Cubic>, <at 1 0>]}}}
     Scale Curve=&FloatFromCurve(
      FloatCurve=DataInterface() {propertyValues {Curve {keys [{arriveTangent=-0.0014, interpMode=RCIM_Cubic, time=0, value=1}, <at 500 0 RCIM_Cubic>]}}}
      CurveIndex=&DistanceBetweenPositions(A=$System.PlayerPosition, B=$Particles.Position)
     )
    )
    ScaleA=true
    ScaleRGB=false
   )
  ]
  rendererClasses [&NiagaraSpriteRendererProperties]
  renderers [{rendererClass=&NiagaraSpriteRendererProperties}]
 }
}
summary {
 systemName=NS_LimeRock_Barrier_Tree
 emitters [
  {emitterName=Shimmers, rendererClasses [&NiagaraSpriteRendererProperties]}
  {emitterName=Glow, rendererClasses [&NiagaraSpriteRendererProperties]}
  {emitterName=AmbientMotes, simTarget=GPUComputeSim, rendererClasses [&NiagaraSpriteRendererProperties]}
  {emitterName=Leaves, simTarget=GPUComputeSim, rendererClasses [&NiagaraSpriteRendererProperties]}
  {
   bEnabled=false
   emitterName=BarrierRibbon
   rendererClasses [&NiagaraSpriteRendererProperties, &NiagaraRibbonRendererProperties]
  }
  {emitterName=BarrierMesh, rendererClasses [&NiagaraMeshRendererProperties]}
 ]
 userVariables [
  User.Intensity {type=NiagaraFloat, defaultValue=0, description=""}
  User.PointForce {type=NiagaraFloat, defaultValue=0, description=""}
  User.PointForceOrigin {type=NiagaraPosition, description="", defaultValue {}}
 ]
}
